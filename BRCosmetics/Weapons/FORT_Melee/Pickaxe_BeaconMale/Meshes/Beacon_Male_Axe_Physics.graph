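# Physics asset for the Beacon Male two-handed pickaxe, defining simulation bodies for its handle and dangling tassel.
# One capsule covers the handle bone, and five capsules follow the dyn_tassle bone chain.
# Constraints link each tassel segment to the one before it, with the first attached to the handle.
# The root joint allows wider, asymmetric swing, while the later joints share tighter swing limits and all share the same twist limit.
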PhysicsAsset /BRCosmetics/Weapons/FORT_Melee/Pickaxe_BeaconMale/Meshes/Beacon_Male_Axe_Physics
bodies {
 Melee_TwoHanded_Handle [{shapeType=Capsule, length=26.9566, radius=4.9885, center=<-2.7622 3.6051 -27.1152>, rotation=<rot -7.3169 -89.7718 -1.2099>}]
 dyn_tassle_01 [{shapeType=Capsule, length=5.6115, radius=1.8901, center=<4.336 0.1008 0.0382>, rotation=<rot 0.137 88.9633 -89.0217>}]
 dyn_tassle_02 [{shapeType=Capsule, length=6.8295, radius=1.8173, center=<3.7349 0.0379 -0.1269>, rotation=<rot 0 -88.4502 -83.0149>}]
 dyn_tassle_03 [{shapeType=Capsule, length=2.7475, radius=4.5403, center=<4.508 0.4104 0.522>, rotation=<rot 0 -80.2899 -71.9282>}]
 dyn_tassle_04 [{shapeType=Capsule, length=2.8399, radius=2.621, center=<2.2827 0.2856 0.3179>, rotation=<rot -12.4895 -90.5772 -88.8269>}]
 dyn_tassle_05 [{shapeType=Capsule, length=0.5204, radius=3.7055, center=<2.5258 0 0>, rotation=<rot 90 0 0>}]
}
constraints [
 {bone1Name=dyn_tassle_01, bone2Name=Melee_TwoHanded_Handle, swing1LimitDegrees=30, swing2LimitDegrees=45, twistLimitDegrees=15}
 {bone1Name=dyn_tassle_02, bone2Name=dyn_tassle_01, swing1LimitDegrees=22, swing2LimitDegrees=22, twistLimitDegrees=15}
 {bone1Name=dyn_tassle_03, bone2Name=dyn_tassle_02, swing1LimitDegrees=22, swing2LimitDegrees=22, twistLimitDegrees=15}
 {bone1Name=dyn_tassle_04, bone2Name=dyn_tassle_03, swing1LimitDegrees=22, swing2LimitDegrees=22, twistLimitDegrees=15}
 {bone1Name=dyn_tassle_05, bone2Name=dyn_tassle_04, swing1LimitDegrees=22, swing2LimitDegrees=22, twistLimitDegrees=15}
]
